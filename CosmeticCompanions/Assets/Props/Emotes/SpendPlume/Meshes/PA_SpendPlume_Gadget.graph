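# Physics asset for the SpendPlume emote gadget, a jellyfish-shaped cosmetic prop with dangling tentacles.
# A large capsule body covers the main jellyfish bone, with four tentacles (01, 03, 05, 06) each modeled as a chain of four small capsules.
# Each tentacle chain is constrained to the main body and link to link, with uniform swing and twist limits on every joint.
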PhysicsAsset /CosmeticCompanions/Assets/Props/Emotes/SpendPlume/Meshes/PA_SpendPlume_Gadget
bodies {
 jellyfishmain [{shapeType=Capsule, length=1.8, radius=34.9624, shapeName=jellyfishmain_capsule, center=<0 -25.5593 0>, rotation=<rot 0 -90 0>}]
 tentacle_01_01 [{
   shapeType=Capsule
   length=7.3189
   radius=5.0796
   shapeName=tentacle_01_01_capsule
   center=<0.1338 0.0014 -8.6773>
   rotation=<rot -0.0003 -89.4088 179.1076>
  }]
 tentacle_01_02 [{
   shapeType=Capsule
   length=7.3217
   radius=5.0824
   shapeName=tentacle_01_02_capsule
   center=<0.1337 -0.0005 -8.6433>
   rotation=<rot 0.0031 -90.003 179.1076>
  }]
 tentacle_01_03 [{
   shapeType=Capsule
   length=7.3242
   radius=5.0848
   shapeName=tentacle_01_03_capsule
   center=<0.0013 -0.0461 -8.8424>
   rotation=<rot 0.0086 -0.0001 -179.7065>
  }]
 tentacle_01_04 [{
   shapeType=Capsule
   length=5.9611
   radius=5.402
   shapeName=tentacle_01_04_capsule
   center=<0.0071 0.0406 -8.5183>
   rotation=<rot 0.6703 -90 0.1176>
  }]
 tentacle_03_01 [{
   shapeType=Capsule
   length=6.8427
   radius=5.0811
   shapeName=tentacle_03_01_capsule
   center=<-0.0963 -0.0964 -8.7841>
   rotation=<rot 0.0006 135.0694 179.108>
  }]
 tentacle_03_02 [{
   shapeType=Capsule
   length=6.8665
   radius=5.0845
   shapeName=tentacle_03_02_capsule
   center=<0.0219 -0.0332 -8.8754>
   rotation=<rot 0.0527 -135.0003 179.745>
  }]
 tentacle_03_03 [{
   shapeType=Capsule
   length=6.5793
   radius=5.0878
   shapeName=tentacle_03_03_capsule
   center=<0.0058 -0.0165 -8.7807>
   rotation=<rot 0.0492 -135.0001 179.8968>
  }]
 tentacle_03_04 [{
   shapeType=Capsule
   length=5.2526
   radius=5.4013
   shapeName=tentacle_03_04_capsule
   center=<0.0536 0.0536 -8.6025>
   rotation=<rot 0.9239 -90 0.9237>
  }]
 tentacle_05_01 [{
   shapeType=Capsule
   length=7.5114
   radius=5.0811
   shapeName=tentacle_05_01_capsule
   center=<-0.108 -0.0766 -8.5248>
   rotation=<rot 0 125.3459 179.1074>
  }]
 tentacle_05_02 [{
   shapeType=Capsule
   length=6.6145
   radius=5.0845
   shapeName=tentacle_05_02_capsule
   center=<-0.0776 0.1116 -8.8479>
   rotation=<rot 0.0028 34.9919 179.108>
  }]
 tentacle_05_03 [{
   shapeType=Capsule
   length=7.5182
   radius=5.0878
   shapeName=tentacle_05_03_capsule
   center=<-0.1097 -0.0768 -8.6892>
   rotation=<rot 0 124.9977 179.1142>
  }]
 tentacle_05_04 [{
   shapeType=Capsule
   length=6.2881
   radius=5.3195
   shapeName=tentacle_05_04_capsule
   center=<0.0621 0.0435 -8.6025>
   rotation=<rot 0.7494 -90 1.0701>
  }]
 tentacle_06_01 [{
   shapeType=Capsule
   length=6.6953
   radius=5.0797
   shapeName=tentacle_06_01_capsule
   center=<-0.1352 0.0306 -8.9665>
   rotation=<rot -0.0002 77.239 179.1085>
  }]
 tentacle_06_02 [{
   shapeType=Capsule
   length=7.6132
   radius=5.0834
   shapeName=tentacle_06_02_capsule
   center=<-0.1314 0.0272 -8.713>
   rotation=<rot -0.0002 78.2795 179.1085>
  }]
 tentacle_06_03 [{
   shapeType=Capsule
   length=7.0759
   radius=5.0864
   shapeName=tentacle_06_03_capsule
   center=<-0.1379 0.0287 -8.9846>
   rotation=<rot -0.0001 78.2267 179.1085>
  }]
 tentacle_06_04 [{
   shapeType=Capsule
   length=5.8303
   radius=5.3198
   shapeName=tentacle_06_04_capsule
   center=<0.0155 0.075 -8.5485>
   rotation=<rot 1.26 -90 0.2602>
  }]
}
constraints [
 {bone1Name=tentacle_01_02, bone2Name=tentacle_01_01, swing1LimitDegrees=45, swing2LimitDegrees=45, twistLimitDegrees=45}
 {bone1Name=tentacle_01_03, bone2Name=tentacle_01_02, swing1LimitDegrees=45, swing2LimitDegrees=45, twistLimitDegrees=45}
 {bone1Name=tentacle_01_04, bone2Name=tentacle_01_03, swing1LimitDegrees=45, swing2LimitDegrees=45, twistLimitDegrees=45}
 {bone1Name=tentacle_03_02, bone2Name=tentacle_03_01, swing1LimitDegrees=45, swing2LimitDegrees=45, twistLimitDegrees=45}
 {bone1Name=tentacle_03_03, bone2Name=tentacle_03_02, swing1LimitDegrees=45, swing2LimitDegrees=45, twistLimitDegrees=45}
 {bone1Name=tentacle_03_04, bone2Name=tentacle_03_03, swing1LimitDegrees=45, swing2LimitDegrees=45, twistLimitDegrees=45}
 {bone1Name=tentacle_05_02, bone2Name=tentacle_05_01, swing1LimitDegrees=45, swing2LimitDegrees=45, twistLimitDegrees=45}
 {bone1Name=tentacle_05_03, bone2Name=tentacle_05_02, swing1LimitDegrees=45, swing2LimitDegrees=45, twistLimitDegrees=45}
 {bone1Name=tentacle_05_04, bone2Name=tentacle_05_03, swing1LimitDegrees=45, swing2LimitDegrees=45, twistLimitDegrees=45}
 {bone1Name=tentacle_06_02, bone2Name=tentacle_06_01, swing1LimitDegrees=45, swing2LimitDegrees=45, twistLimitDegrees=45}
 {bone1Name=tentacle_06_03, bone2Name=tentacle_06_02, swing1LimitDegrees=45, swing2LimitDegrees=45, twistLimitDegrees=45}
 {bone1Name=tentacle_06_04, bone2Name=tentacle_06_03, swing1LimitDegrees=45, swing2LimitDegrees=45, twistLimitDegrees=45}
 {bone1Name=tentacle_01_01, bone2Name=jellyfishmain, swing1LimitDegrees=45, swing2LimitDegrees=45, twistLimitDegrees=45}
 {bone1Name=tentacle_03_01, bone2Name=jellyfishmain, swing1LimitDegrees=45, swing2LimitDegrees=45, twistLimitDegrees=45}
 {bone1Name=tentacle_05_01, bone2Name=jellyfishmain, swing1LimitDegrees=45, swing2LimitDegrees=45, twistLimitDegrees=45}
 {bone1Name=tentacle_06_01, bone2Name=jellyfishmain, swing1LimitDegrees=45, swing2LimitDegrees=45, twistLimitDegrees=45}
]
